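; Rift gate base actor: a portal that marks itself for players and fires a success event when a non-bot pawn enters.
; BeginPlay opens it once playlist data loads, using timeline-driven Niagara, material and audio parameters with a spline-following zipper effect.
; The overlap capsule allows one successful interaction, with authority optionally destroying the rift; a repeating timer removes it outside the safe zone.
; A custom event knocks back nearby players and vehicles using hotfixable values.
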(event Custom|DestroyRift
  (bind _returnvalue_location (Math|Transform|BreakTransform (Transformation|GetWorldTransform (Variables|Default|GetSM_Rift))))
  (bind _portal_inner (Variables|Default|GetPortal_Inner))
  ;[1 1 1] On Death
  ;;[1 0 0 0.5] CLOSING PORTAL VFX (WIP) ISNT WORKING CURRENTLY
  ;;;[1 1 1] DESPAWN PORTAL FX
  (bind _returnvalue (Niagara|SpawnSystematLocation "/RiftGateBase/Gameplay/FX/NS_RiftGate_Portal_Destroy.NS_RiftGate_Portal_Destroy" _returnvalue_location))
  (Effects|Components|ParticleSystem|SetFloatParameter _returnvalue "PortalSize" 2.5)
  (bind (_direction _spawninfluence _zipper_curve _openingcurve _spawninfinner) (Timeline|Timeline_0 ReverseFromEnd)
    (:Update
      (Niagara|SetNiagaraVariable(Float) (Variables|Default|GetNS_Portal) "OpenAmount" _openingcurve)
      (Rendering|Material|SetScalarParameterValueonMaterials _portal_inner "OpenAmount" _openingcurve)
      (Rendering|Material|SetScalarParameterValueonMaterials (Variables|Default|GetPortal_Outer) "SpawnInfluence" _spawninfluence)
      (Rendering|Material|SetScalarParameterValueonMaterials _portal_inner "Edge Opacity G Mask Mult" (* _spawninfinner 1e+02))
      (Rendering|Material|SetScalarParameterValueonMaterials (Variables|Default|GetPortal_Back) "OpcityMult" _spawninfinner)
      (Niagara|SetNiagaraVariable(Float) (Variables|Default|GetNS_RiftGate_Beam) "OpenAmount" _spawninfinner))
    (:Finished
      (CallFunction|DestroyRiftInternal))))

(event Custom|DestroyRift_Internal
  (bind _audio_looping (Variables|Sound|GetAudio_Looping))
  ;[1 1 1] On Death
  (Variables|RiftPortal|SetPortalOpen false)
  (Utilities|FlowControl|Sequence
    (:then_0
      (Networking|CanExecuteCosmeticEvents
        (:True
          (Development|SetHiddeninGame (Variables|Default|GetSM_Rift) true false)
          (Audio|PlaySoundatLocation (Variables|Sound|GetRiftDestroyed) (Transformation|GetActorLocation)) ; Play Deactivate Sound
          (bind (_direction _pitchcurve) (Timeline|Audio%20RiftDestroyed Play)
            (:Update
              (Audio|Components|Audio|SetPitchMultiplier _audio_looping _pitchcurve))
            (:Finished
              (Audio|Components|Audio|FadeOut _audio_looping 1 0.0 "Linear" :!class /Script/Engine.AudioComponent))))))
    (:then_1
      (Utilities|FlowControl|Delay (ScalableFloat|GetValueatLevel (Variables|HotfixRows|GetDelayToDestroy)))
      (Actor|DestroyActor))))

(event EventBeginPlay
  (bind _sm_rift (Variables|Default|GetSM_Rift))
  (bind _returnvalue_location (Math|Transform|BreakTransform (Transformation|GetWorldTransform _sm_rift)))
  (bind _returnvalue (+ _returnvalue_location "0.000000,0.000000,475.000000"))
  (bind _portal_inner (Variables|Default|GetPortal_Inner))
  (bind _ns_portalzipper (Variables|Default|GetNS_PortalZipper))
  ;[1 1 1] EVENT--> BeginPlay
  (|CallParentFunction)
  (Variables|RiftPortal|SetPortalOpen true)
  (Utilities|FlowControl|Sequence
    (:then_0
      ;[1 0 0 0.5] SCALABILITY TWEAKS TO FORCE TURN OFF AMBIENT LIGHT ON MEDIUM AND LOW PLATFORM
      (bind _returnvalue_2 (Settings|GetGameUserSettings))
      (switch int (Settings|GetVisualEffectQuality _returnvalue_2)
        (:0
          _)
        (:1
          _))
      (Rendering|SetVisibility (Variables|Default|GetAmbient_Light) false true))
    (:then_1
      (Utilities|Time|SetTimerbyEvent (delegate Custom|CheckForStorm) 3.0 true)
      ;[1 0 0 0.5] SET THE PORTAL MATERIAL PARAMETER TO ZERO (VFX)
      (Rendering|Material|SetScalarParameterValueonMaterials (Variables|Default|GetPortal_Inner) "OpenAmount" 0.0)
      ;[1 1 1] Check if In Creative Mode or not
      (bind _asfort_game_state_athena (Utilities|Casting|CastToFortGameStateAthena (Game|GetGameState)))
      (if (Variables|/Script/FortniteGame.FortGameStateAthena|GetbPlaylistDataIsLoaded _asfort_game_state_athena)
        ;;[0.3843 0.151 1] UI Ping and Compass - Distance Checks
        (HUDPings|MarkActor self) ; HUD Ping
        (Utilities|FlowControl|Sequence
          (:then_0
            ;;[1 0 0 0.5] OPENING PORTAL VFX
            (bind _cancompletequest (Class|AthenaQuestFunctionLibrary|AthenaCheckCreativeMode (Game|GetGameState)))
            (if _cancompletequest
              (Variables|QuestStuff|SetCreativeMode false)
              (else
                (Variables|QuestStuff|SetCreativeMode true))))
          (:then_1
            ;;;[1 1 1] SPAWN PORTAL FX
            (bind _returnvalue_3 (Niagara|SpawnSystematLocation "/RiftGateBase/Gameplay/FX/NS_RiftGate_Portal_Spawning.NS_RiftGate_Portal_Spawning" _returnvalue))
            (Effects|Components|ParticleSystem|SetFloatParameter _returnvalue_3 "Lifetime" 2.25)
            (Effects|Components|ParticleSystem|SetFloatParameter _returnvalue_3 "GlobalScale" 1e+02)
            (Effects|Components|ParticleSystem|SetFloatParameter _returnvalue_3 "PortalSize" 3.25)
            (Utilities|FlowControl|Delay 0.9)
            (bind _returnvalue_4 (Niagara|SpawnSystematLocation "/RiftGateBase/Gameplay/FX/NS_Portal_Zipper.NS_Portal_Zipper" (Spline|GetLocationatSplinePoint (Variables|Default|GetPortal_Spline) (Math|Integer|SelectInt 4) "World") (Spline|GetRotationatSplinePoint (Variables|Default|GetPortal_Spline) (Math|Integer|SelectInt 4) "World")))
            (bind _output_get (Variables|Default|SetNS_PortalZipper _returnvalue_4))
            (Rendering|SetTranslucentSortPriority _output_get 2)
            (Rendering|Material|SetScalarParameterValueonMaterials (Variables|Default|GetPortal_Back) "OpacityMult" 0.0)
            (Rendering|Material|SetScalarParameterValueonMaterials (Variables|Default|GetPortal_Inner) "SpawnInfluence" 0.0)
            (Rendering|ToggleVisibility (Variables|Default|GetPortal_Outer) false)
            (Niagara|SpawnSystematLocation "/RiftGateBase/Gameplay/FX/NS_RiftGate_SpawnImpact.NS_RiftGate_SpawnImpact" (+ _returnvalue_location "0.000000,0.000000,-100.000000"))
            (bind (_direction _spawninfluence _zipper_curve _openingcurve _spawninfinner) (Timeline|Timeline PlayFromStart)
              (:Update
                (Utilities|IsValid _ns_portalzipper
                  (:Is%20Valid
                    (Transformation|SetWorldTransform _ns_portalzipper (Math|Transform|MakeTransform (Spline|GetLocationatDistanceAlongSpline (Variables|Default|GetPortal_Spline) (* _zipper_curve (Spline|GetSplineLength (Variables|Default|GetPortal_Spline))) "World") (Spline|GetRotationatDistanceAlongSpline (Variables|Default|GetPortal_Spline) (* _zipper_curve (Spline|GetSplineLength (Variables|Default|GetPortal_Spline))) "World")) false false)
                    (Audio|Parameter|SetFloatParameter (Variables|Sound|GetAudio_Looping) "OpenAmount" _openingcurve :!class /Script/AudioExtensions.AudioParameterControllerInterface)
                    (Niagara|SetNiagaraVariable(Float) (Variables|Default|GetNS_Portal) "OpenAmount" _openingcurve)
                    (Rendering|Material|SetScalarParameterValueonMaterials _portal_inner "OpenAmount" _openingcurve)
                    (Rendering|Material|SetScalarParameterValueonMaterials (Variables|Default|GetPortal_Outer) "SpawnInfluence" _spawninfluence)
                    (Rendering|Material|SetScalarParameterValueonMaterials _portal_inner "Edge Opacity G Mask Mult" (* _spawninfinner 1e+02))
                    (Niagara|SetNiagaraVariable(Float) (Variables|Default|GetNS_RiftGate_Beam) "OpenAmount" _spawninfinner)
                    (Rendering|Material|SetScalarParameterValueonMaterials (Variables|Default|GetPortal_Back) "OpacityMult" _spawninfinner))))))
          (:then_2
            (Audio|SpawnSoundatLocation (Variables|Sound|GetPortalOpenSound) _returnvalue)))
        (else
          (Playlist|BindEventtoOnCurrentPlaylistLoaded _asfort_game_state_athena (delegate Custom|PlaylistLoaded)))))))

(event Custom|PlaylistLoaded
  (params (PlaylistName :type Name) (PlaylistContextTags :type Gameplay%20Tag%20Container%20Structure%20%28by%20ref%29))
  (bind _sm_rift (Variables|Default|GetSM_Rift))
  (bind _returnvalue_location (Math|Transform|BreakTransform (Transformation|GetWorldTransform _sm_rift)))
  (bind _returnvalue (+ _returnvalue_location "0.000000,0.000000,475.000000"))
  (bind _portal_inner (Variables|Default|GetPortal_Inner))
  (bind _ns_portalzipper (Variables|Default|GetNS_PortalZipper))
  ;[1 1 1] Check if In Creative Mode or not
  ;;[0.3843 0.151 1] UI Ping and Compass - Distance Checks
  (HUDPings|MarkActor self) ; HUD Ping
  (Utilities|FlowControl|Sequence
    (:then_0
      ;;[1 0 0 0.5] OPENING PORTAL VFX
      (bind _cancompletequest (Class|AthenaQuestFunctionLibrary|AthenaCheckCreativeMode (Game|GetGameState)))
      (if _cancompletequest
        (Variables|QuestStuff|SetCreativeMode false)
        (else
          (Variables|QuestStuff|SetCreativeMode true))))
    (:then_1
      ;;;[1 1 1] SPAWN PORTAL FX
      (bind _returnvalue_2 (Niagara|SpawnSystematLocation "/RiftGateBase/Gameplay/FX/NS_RiftGate_Portal_Spawning.NS_RiftGate_Portal_Spawning" _returnvalue))
      (Effects|Components|ParticleSystem|SetFloatParameter _returnvalue_2 "Lifetime" 2.25)
      (Effects|Components|ParticleSystem|SetFloatParameter _returnvalue_2 "GlobalScale" 1e+02)
      (Effects|Components|ParticleSystem|SetFloatParameter _returnvalue_2 "PortalSize" 3.25)
      (Utilities|FlowControl|Delay 0.9)
      (bind _returnvalue_3 (Niagara|SpawnSystematLocation "/RiftGateBase/Gameplay/FX/NS_Portal_Zipper.NS_Portal_Zipper" (Spline|GetLocationatSplinePoint (Variables|Default|GetPortal_Spline) (Math|Integer|SelectInt 4) "World") (Spline|GetRotationatSplinePoint (Variables|Default|GetPortal_Spline) (Math|Integer|SelectInt 4) "World")))
      (bind _output_get (Variables|Default|SetNS_PortalZipper _returnvalue_3))
      (Rendering|SetTranslucentSortPriority _output_get 2)
      (Rendering|Material|SetScalarParameterValueonMaterials (Variables|Default|GetPortal_Back) "OpacityMult" 0.0)
      (Rendering|Material|SetScalarParameterValueonMaterials (Variables|Default|GetPortal_Inner) "SpawnInfluence" 0.0)
      (Rendering|ToggleVisibility (Variables|Default|GetPortal_Outer) false)
      (Niagara|SpawnSystematLocation "/RiftGateBase/Gameplay/FX/NS_RiftGate_SpawnImpact.NS_RiftGate_SpawnImpact" (+ _returnvalue_location "0.000000,0.000000,-100.000000"))
      (Timeline|Timeline PlayFromStart))
    (:then_2
      (Audio|SpawnSoundatLocation (Variables|Sound|GetPortalOpenSound) _returnvalue))))

(event Custom|CheckForStorm
  (bind _returnvalue (SafeZone|IsLocationinSafeZone (Transformation|GetActorLocation)))
  (if (not _returnvalue)
    (Utilities|FlowControl|DoOnce false
      (:Completed
        (HUDPings|UnmarkActor)
        (Actor|DestroyActor)))))

(event OnComponentBeginOverlap(Capsule)
  (params (OverlappedComponent :type Primitive%20Component%20Object%20Reference) (OtherActor :type Actor%20Object%20Reference) (OtherComp :type Primitive%20Component%20Object%20Reference) (OtherBodyIndex :type Integer) (bFromSweep :type Boolean) (SweepResult :type Hit%20Result%20Structure%20%28by%20ref%29))
  (bind _sm_rift (Variables|Default|GetSM_Rift))
  ;[1 1 1] On Overlap - Toggle Post Process
  (Utilities|Casting|CastToFortPlayerPawn OtherActor)
  (CallFunction|PlayExpandorCollapseAudio true)
  (bind (_direction _open) (Timeline|VectorScale Play)
    (:Update
      (Rendering|Material|SetScalarParameterValueonMaterials _sm_rift "Open" _open)
      (Transformation|SetRelativeRotation _sm_rift (Math|Rotator|MakeRotator (* _open -2e+01)) false false)
      (Effects|Components|ParticleSystem|SetFloatParameter (Variables|Default|GetNS_Rift_Portal_Idle_Shimmer) "Open" _open))))

(event OnComponentEndOverlap(Capsule)
  (params (OverlappedComponent :type Primitive%20Component%20Object%20Reference) (OtherActor :type Actor%20Object%20Reference) (OtherComp :type Primitive%20Component%20Object%20Reference) (OtherBodyIndex :type Integer))
  (bind _sm_rift (Variables|Default|GetSM_Rift))
  ;[1 1 1] On Overlap - Toggle Post Process
  (if (<= (Utilities|Array|Length (Collision|GetOverlappingActors OtherActor "/Script/FortniteGame.FortPlayerPawn")) 0)
    (CallFunction|PlayExpandorCollapseAudio false)
    (Timeline|VectorScale Reverse)))

(event EventOverlappingOnBeginPlay
  (params (Actor :type Actor%20Object%20Reference))
  ;[1 1 1] Perform rift interaction
  (Variables|RiftPortal|SetOverlapActor Actor)
  (if (and (Variables|RiftPortal|GetPortalOpen) (not (Variables|Default|GetInteractionPerformed)))
    (bind _asfort_pawn (Utilities|Casting|CastToFortPawn (Variables|RiftPortal|GetOverlapActor)))
    (if (not (Pawn|IsBotControlled _asfort_pawn))
      (CallFunction|ResetFX)
      (bind _succeeded (CallFunction|PerformInteraction))
      (bind _output_get (Variables|Default|SetInteractionPerformed _succeeded))
      (if _output_get
        (Networking|SwitchHasAuthority
          (:Authority
            (CallFunction|NotifySuccessfulInteraction (Transformation|GetActorLocation (Variables|RiftPortal|GetPlayerPawn)))
            (Default|CallOnSuccessfulInteraction)
            (if (Variables|RiftPortal|GetbDestroyOnInteract)
              (CallFunction|DestroyRift))))))))

(event Custom|KnockBackPlayers
  ;[1 1 1] Knockback Pawns
  (bind _outhits (Collision|MultiSphereTraceForObjects (Transformation|GetActorLocation) (+ (Transformation|GetActorLocation) "0.000000,0.000000,50.000000") (ScalableFloat|GetValueatLevel (Variables|HotfixRows|GetKnockbackRadiusRow)) (Variables|KnockbackPlayers|GetKnockbackObjectToDetect)))
  (for _array_element _outhits
    (bind (_array_element_bblockinghit _array_element_binitialoverlap _array_element_time _array_element_distance _array_element_location _array_element_impactpoint _array_element_normal _array_element_impactnormal _array_element_physmat _array_element_hitactor) (Collision|BreakHitResult _array_element))
    (bind _asfort_player_pawn (Utilities|Casting|CastToFortPlayerPawn _array_element_hitactor)
      (:then
        (Pawn|Character|LaunchCharacterJump _asfort_player_pawn (CallFunction|CalcLaunchDirection _asfort_player_pawn (ScalableFloat|GetValueatLevel (Variables|HotfixRows|GetKnockbackMagnitudeRow)) (ScalableFloat|GetValueatLevel (Variables|HotfixRows|GetKnockbackZ_MagnitudeRow))) true true true true)
        (GameplayCue|ExecuteGameplayCueOnActor(Burst) _asfort_player_pawn (Variables|Default|GetRiftKnockback) (Ability|GameplayCue|MakeGameplayCueParameters 0.0)))
      (:CastFailed
        (bind _asfort_vehicle_interface (Utilities|Casting|CastToFortVehicleInterface _array_element_hitactor))
        (Audio|BroadcastAbilityImpactatLocation 0 (Utilities|ToObject(Interface) _asfort_vehicle_interface) (CallFunction|CalcLaunchDirection _array_element_hitactor (ScalableFloat|GetValueatLevel (Variables|HotfixRows|GetKnockbackVehicles_Magnitude)) (ScalableFloat|GetValueatLevel (Variables|HotfixRows|GetKnockbackVehicles_MagnitudeZ))) 7.8e+02) ; This is the secret node that launches vehicles. Note it needs the 780 radius to work.
      ))))

(event OnComponentBeginOverlap(OverlapCapsule)
  (params (OverlappedComponent :type Primitive%20Component%20Object%20Reference) (OtherActor :type Actor%20Object%20Reference) (OtherComp :type Primitive%20Component%20Object%20Reference) (OtherBodyIndex :type Integer) (bFromSweep :type Boolean) (SweepResult :type Hit%20Result%20Structure%20%28by%20ref%29))
  ;[1 1 1] Perform rift interaction
  (Variables|RiftPortal|SetOverlapActor OtherActor)
  (if (and (Variables|RiftPortal|GetPortalOpen) (not (Variables|Default|GetInteractionPerformed)))
    (bind _asfort_pawn (Utilities|Casting|CastToFortPawn (Variables|RiftPortal|GetOverlapActor)))
    (if (not (Pawn|IsBotControlled _asfort_pawn))
      (CallFunction|ResetFX)
      (bind _succeeded (CallFunction|PerformInteraction))
      (bind _output_get (Variables|Default|SetInteractionPerformed _succeeded))
      (if _output_get
        (Networking|SwitchHasAuthority
          (:Authority
            (CallFunction|NotifySuccessfulInteraction (Transformation|GetActorLocation (Variables|RiftPortal|GetPlayerPawn)))
            (Default|CallOnSuccessfulInteraction)
            (if (Variables|RiftPortal|GetbDestroyOnInteract)
              (CallFunction|DestroyRift))))))))

(event Custom|NotifySuccessfulInteraction
  (params (Location :type Vector))
  ;[1 1 1] Spawn rift interaction visuals and audio for clients
  (Utilities|FlowControl|DoOnce false
    (:Completed
      (CallFunction|CreateInteractionFX Location))))

(event Custom|CreateInteractionFX
  (params (Location :type Vector))
  ;[1 1 1] Spawn rift interaction visuals and audio for clients
  (Effects|Components|ParticleSystem|SpawnFXAtLocation (Variables|RiftPortal|GetEnterRift_NS) Location)
  (if (not (Utilities|Equal(Object) (Variables|RiftPortal|GetOverlapActor) (Game|GetPlayerPawn 0)))
    (Audio|PlaySoundatLocation (Variables|Sound|GetEnteredRift) Location)))

(event Custom|ResetFX
  ;[1 1 1] Spawn rift interaction visuals and audio for clients
  (Utilities|FlowControl|DoOnce false
    (:Completed
      (CallFunction|CreateInteractionFX (stale Custom|NotifySuccessfulInteraction Location)))))

(fn BlockInteractionsWIthDisguises
  (params (Actor :type Actor%20Object%20Reference))
  (GameplayEffects|ApplyGameplayEffectToSelf (Ability|GetFortAbilitySystemComponent Actor) "/Game/Athena/Items/ForagedItems/Rift/GE_ProhibitDisguiseInteract.GE_ProhibitDisguiseInteract_C" 0.0 (GameplayEffects|MakeEffectContext (Ability|GetFortAbilitySystemComponent Actor))))

(fn CalcLaunchDirection
  (params (Actor :type Actor%20Object%20Reference) (Magnitude :type Float%20%28double-precision%29) (AddedZ_Magnitude :type Float%20%28double-precision%29))
  (return (Math|Vector|MakeVector (.x (* (Math|Vector|Normalize (- (Transformation|GetActorLocation Actor) (Transformation|GetActorLocation))) Magnitude)) (.y (* (Math|Vector|Normalize (- (Transformation|GetActorLocation Actor) (Transformation|GetActorLocation))) Magnitude)) (Math|Float|Max(Float) (+ (.z (* (Math|Vector|Normalize (- (Transformation|GetActorLocation Actor) (Transformation|GetActorLocation))) Magnitude)) 1.0) AddedZ_Magnitude))))

(fn IsFlying
  (bind _playerpawn (Variables|RiftPortal|GetPlayerPawn))
  (Utilities|IsValid _playerpawn
    (:Is%20Valid
      (bind _asfort_movement_comp_character (Utilities|Casting|CastToFortMovementComp_Character (Pawn|GetMovementComponent _playerpawn))
        (:then
          (return (Utilities|Enum|Equal(Enum) (Variables|/Script/Engine.CharacterMovementComponent|GetMovementMode _asfort_movement_comp_character) "MOVE_Flying")))
        (:CastFailed
          (return false))))))

(fn IsPlayerAttachedToAIPawn
  (params (AIPawn :type Fort%20AIPawn%20Object%20Reference))
  (for _array_element (Actor|GetAttachedActors AIPawn true false)
    (bind _asfort_player_pawn (Utilities|Casting|CastToFortPlayerPawn _array_element))
    (return true _asfort_player_pawn))
  (return false 0))

(fn MarkActor
  (params (ActorToMark :type Actor%20Object%20Reference))
  (if (Variables|RiftPortal|GetbShouldMarkRift)
    (for _array_element (Fortnite|GetAllFortPlayerControllers true true)
      (Utilities|IsValid _array_element
        (:Is%20Valid
          (Utilities|IsValid (Class|FortPlayerController|GetMarkerComponent _array_element)
            (:Is%20Valid
              (Utilities|IsValid ActorToMark
                (:Is%20Valid
                  (bind _returnvalue (MarkerComponent|MarkActoronServer (Class|FortPlayerController|GetMarkerComponent _array_element) ActorToMark false false))
                  (Variables|Default|SetPortalMarkerID _returnvalue))))))))))

(fn OnRep_ActivePawn)

(fn PerformInteraction
  (bind _overlapactor (Variables|RiftPortal|GetOverlapActor))
  (Utilities|IsValid _overlapactor
    (:Is%20Valid
      (bind _asfort_player_pawn (Utilities|Casting|CastToFortPlayerPawn _overlapactor))
      (Variables|RiftPortal|SetPlayerPawn _asfort_player_pawn)
      (return true))))

(fn PlayExpandOrCollapseAudio
  (params (Expand :type Boolean))
  (if (Variables|RiftPortal|GetPortalOpen)
    (Audio|PlaySoundatLocation (select Expand "/Game/Sounds/Fort_GamePlay_Sounds/Player/RiftPortal/RiftPortal_Expand_Cue.RiftPortal_Expand_Cue" "/Game/Sounds/Fort_GamePlay_Sounds/Player/RiftPortal/RiftPortal_Collapse_Cue.RiftPortal_Collapse_Cue") (Transformation|GetWorldLocation (Variables|Default|GetSM_Rift)))))

(fn UnmarkActor
  (bind _slatebrush (Utilities|Struct|MakeSlateBrush "(SpecifiedColor=(R=1.000000,G=1.000000,B=1.000000,A=1.000000),ColorUseRule=UseColor_Specified)" "Image" "NoTile" "NoMirror" (struct :X 0.0 :Y 0.0) "(Left=0.000000,Top=0.000000,Right=0.000000,Bottom=0.000000)" 0 "(CornerRadii=(X=0.000000,Y=0.000000,Z=0.000000,W=0.000000),Color=(SpecifiedColor=(R=0.000000,G=0.000000,B=0.000000,A=0.000000),ColorUseRule=UseColor_Specified),Width=0.000000,RoundingType=HalfHeightRadius,bUseBrushTransparency=False)"))
  (bind _athenaspecialactor (Variables|Default|GetAthenaSpecialActor))
  (if (Variables|RiftPortal|GetbShouldMarkRift)
    (for _array_element (Fortnite|GetAllFortPlayerControllers true true)
      (Utilities|IsValid _array_element
        (:Is%20Valid
          (Utilities|IsValid (Class|FortPlayerController|GetMarkerComponent _array_element)
            (:Is%20Valid
              (MarkerComponent|UnmarkActoronClient (Class|FortPlayerController|GetMarkerComponent _array_element) (Variables|Default|GetPortalMarkerID)))))))
    (Variables|/Script/FortniteGame.AthenaSpecialActorComponent|SetSpecialActorCompassIconBrush _slatebrush _athenaspecialactor) ; Remove compass icon
    (Variables|/Script/FortniteGame.AthenaSpecialActorComponent|SetSpecialActorMinimapIconBrush _slatebrush _athenaspecialactor) ; Remove minimap icon
  ))

(fn ConstructionScript
  (bind (_returnvalue_x _returnvalue_y) (Math|Vector|BreakVector (Transformation|GetActorLocation)))
  (Effects|Components|ParticleSystem|SetVectorParameter (Variables|Default|GetNS_RiftGate_Beam) "EndPosition" (Math|Vector|MakeVector _returnvalue_x _returnvalue_y 2e+04))
  (bind _returnvalue (Rendering|Material|CreateDynamicMaterialInstance (Variables|Default|GetPortal_Inner) 0 (Variables|Materials|GetInnerPortalMaterial) "None" :!class /Script/Engine.PrimitiveComponent))
  (Variables|Materials|SetMID_Portal_Inner _returnvalue)
  (bind _returnvalue_2 (Rendering|Material|CreateDynamicMaterialInstance (Variables|Default|GetPortal_Back) 0 (Variables|Default|GetBack%20Portal%20Material) "None" :!class /Script/Engine.PrimitiveComponent))
  (Variables|Materials|SetMID_Portal_Back _returnvalue_2))
