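; Base interactable Rift Gate actor that spawns, lets a player enter for a cinematic power grant, then despawns.
; Timers drive the spawn and despawn sequences (building destruction, knockback, FX), while timelines animate opening via spline zipper and material parameters.
; Interaction applies gameplay effects for the cinematic; local spectators get a separate view target, camera animations and screen wipe with cleanup bindings.
; An authority timer despawns gates left outside the safe zone once no players are nearby.
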(event EventLandedOnHayStack
  (params (PlayerPawn :type Fort%20Player%20Pawn%20Object%20Reference) (Z%20Velocity%20Mag :type Float%20%28double-precision%29)))

(event EventBeginPlay
  ;[0 1 0] Begin Play
  (|CallParentFunction)
  (CallFunction|InitVars)
  (Networking|SwitchHasAuthority
    (:Authority
      (Rendering|SetActorHiddenInGame true)
      (CallFunction|PlaySpawnSequence)
      (bind _returnvalue (Utilities|Time|SetTimerbyEvent (EventDispatchers|CreateEvent :!event CheckForInSafeZone) (Variables|SafeZone|GetCheckSafeZoneFrequency) true)) ; Check for safe zone
      (Variables|SafeZone|SetCheckSafeZoneTimer _returnvalue)
      (CallFunction|SetRiftGateOpen true)
      (Default|MarkActor self))))

(event Custom|PlaySpawnSequence
  ;[1 0 0.75] Initial Spawn in sequence
  ;;[0 0 0] Play Spawn Sequence
  (CallFunction|PlaySpawnAnticipationFX)
  (Utilities|Time|SetTimerbyEvent (EventDispatchers|CreateEvent :!event Custom|DestroyBuildingsInArea) (ScalableFloat|GetValueatLevel (Variables|RiftGate|Timing|SpawnSequence|GetRow_Delay_DestroyBuildingsInArea))) ; Timing for Destroying Buildings
  (Utilities|Time|SetTimerbyEvent (EventDispatchers|CreateEvent :!event Custom|KnockbackPlayersInArea) (ScalableFloat|GetValueatLevel (Variables|RiftGate|Timing|SpawnSequence|GetRow_Delay_OnSpawnKnockback))) ; Timing to knock back players
  (Utilities|Time|SetTimerbyEvent (EventDispatchers|CreateEvent :!event Custom|PlaySpawnFX) (ScalableFloat|GetValueatLevel (Variables|RiftGate|Timing|SpawnSequence|GetRow_Delay_SpawnFX))) ; Timing for the FX that precede actually spawning in the Rift Gate
  (Utilities|Time|SetTimerbyEvent (EventDispatchers|CreateEvent :!event Custom|SpawnIn) (ScalableFloat|GetValueatLevel (Variables|RiftGate|Timing|SpawnSequence|GetRow_Duration_InitialSpawnAnticipationFX))) ; Timing to Spawn in the Rift Gate
)

(event Custom|PlaySpawnAnticipationFX
  ;[1 0 0.75] Initial Spawn in sequence
  ;;[0 0 0] Play Spawn Anticipation FX - VFX, SFX, Sound Indicator, Camera Shake, Controller Rumble
  (InitialSpawn|StartInitialSpawnAnticipationFX)
  (Utilities|Time|SetTimerbyEvent (EventDispatchers|CreateEvent :!event StopInitialSpawnAnticipationFX) (ScalableFloat|GetValueatLevel (Variables|RiftGate|Timing|SpawnSequence|GetRow_Duration_InitialSpawnAnticipationFX))))

(event Custom|PlaySpawnFX
  ;[1 0 0.75] Initial Spawn in sequence
  ;;[0 0 0] Play Spawn FX
  (Utilities|FlowControl|Sequence
    (:then_0
      (Niagara|SpawnSystematLocation (Variables|FX|Spawn|GetNS_SpawnFX) (FX|GetFXSpawnLocation) (Math|Rotator|MakeRotator 0.0 0.0 (Math|Random|RandomFloatinRange -1.8e+02 1.8e+02)) "1.000000,1.000000,1.000000" true true "None" true)
      (CallFunction|RadialForce)
      (ForceFeedback|SpawnForceFeedbackatLocation (Variables|FX|Spawn|GetForceFeedback_Spawn) (Transformation|GetActorLocation))
      (if (Variables|Default|GetCameraShakeEnabled)
        (Camera|PlayWorldCameraShake (Variables|FX|Spawn|GetCameraShake_Spawn) (Transformation|GetActorLocation) 0.0 5e+03)))
    (:then_1
      (bind _returnvalue (Rendering|Decal|SpawnDecalatLocation (Variables|FX|Spawn|GetDecal_Spawn) "512.000000,256.000000,256.000000" (Transformation|GetActorLocation) "-90.000000,0.000000,0.000000" 3.0))
      (bind _output_get (Variables|Default|SetSpawned%20Decal%20Component _returnvalue))
      (Rendering|Components|Decal|SetFadeOut _output_get 8.0 2.0 true))))

(event Custom|InitVars
  ;[0.2 1 0.2] Init Vars
  (Class|FortAsyncActionCurrentPlaylistReady|PlaylistDataReadyAsync
    (:OnReady
      (Variables|Actor|SetbFindCameraComponentWhenViewTarget true)
      (Variables|Default|SetUseInstantCutEntranceCameraTransition (ScalableFloat|GetValueasBool (Variables|Default|GetRow_UseInstantCutEntranceCameraTransition))))))

(event EventStopHiding
  (params (Pawn :type Fort%20Pawn%20Object%20Reference))
  ;[1 1 1] Event Start/Stop Hiding - Set rift gate open/closed
  (|CallParentFunction Pawn)
  (Utilities|Time|SetTimerbyEvent (EventDispatchers|CreateEvent :!event Custom|PlayDespawnSequence) (ScalableFloat|GetValueatLevel (Variables|RiftGate|Timing|InteractSequence|GetRow_Delay_ClosePortalOnExit)))
  (Variables|Default|SetbPlayerHasExitedPortal true) ; replicate the player exiting to clients
)

(event EventEndPlay
  (params (EndPlayReason :type EEndPlayReason%20Enum :default "Destroyed"))
  ;[1 0.1 0.1] Event End Play
  (|CallParentFunction EndPlayReason)
  (Utilities|FlowControl|Sequence
    (:then_0
      (bind _forcefeedback (Variables|Default|GetForceFeedback)
        (:then
          (Components|Activation|Deactivate _forcefeedback))
        (:else)))
    (:then_1
      (Networking|CanExecuteCosmeticEvents
        (:True
          (if (Variables|Cinematic|Spectating|GetbInCinematicSpectating)
            (CallFunction|CleanupCinematicSpectating)))))))

(event EventStartHiding
  (params (InteractingPawn :type Fort%20Pawn%20Object%20Reference))
  ;[1 1 1] Event Start/Stop Hiding - Set rift gate open/closed
  (|CallParentFunction InteractingPawn))

(event Custom|SpawnIn
  ;[1 0 0.75] Initial Spawn in sequence
  ;;[0 0 0] Spawn In
  (Rendering|SetActorHiddenInGame false))

(event Interaction|EventBlueprintOnInteract
  (params (InteractingPawn :type Fort%20Pawn%20Object%20Reference) (InteractionBeingAttempted :type EInteractionBeingAttempted%20Enum :default "FirstInteraction"))
  ;[1 1 1] Event On Interact
  (if (Interaction|BlueprintCanInteract InteractingPawn InteractionBeingAttempted)
    (if (not (GameplayTags|HasMatchingGameplayTag InteractingPawn (Variables|Tags|GetTag_PawnHasInteracted)))
      (bind _output_get (Variables|Default|SetLastPawnToInteract InteractingPawn))
      (Variables|RiftGate|Orientation|SetBeingGrantedPowersPawn _output_get)
      (switch Utilities|FlowControl|Switch|SwitchonEInteractionBeingAttempted InteractionBeingAttempted
        (:FirstInteraction
          (Utilities|Casting|CastToFortPlayerController (Pawn|GetController (Variables|Default|GetLastPawnToInteract)))
          (Networking|ForceNetUpdate)
          (CallFunction|GainForcePowers))))))

(event Custom|GainForcePowers
  ;[0.2 0.2 1] Gain Force Powers
  (bind _lastpawntointeract (Variables|Default|GetLastPawnToInteract)
    (:then
      (if (not (AI|Components|NavMovement|IsFalling (Variables|/Script/Engine.Character|GetCharacterMovement _lastpawntointeract)))
        (Utilities|FlowControl|Sequence
          (:then_0
            (Variables|Interaction|SetbAllowInteract false)
            (Variables|Interaction|ControlVars|SetHasGrantedTraining true)
            (Networking|ForceNetUpdate))
          (:then_1
            (CallFunction|PlayCinematicSequence)
            (CallFunction|PlayRiftMusic (Variables|Sound|Music|GetMusic_EnterSequence))
            (Default|UnmarkActor)))))
    (:else)))

(event Custom|PlayEntranceCinematicForSpectators
  (params (PawnInCinematic :type Fort%20Player%20Pawn%20Object%20Reference))
  (bind (_returnvalue_location _returnvalue_rotation) (Math|Transform|BreakTransform (Transformation|GetWorldTransform (Variables|Default|GetHideLocation_ForwardVector))))
  (bind (_returnvalue_rotation_roll _returnvalue_rotation_pitch _returnvalue_rotation_yaw) (Math|Rotator|BreakRotator _returnvalue_rotation))
  ;[0 0 0] Handle spectating entrance cinematics
  (Networking|CanExecuteCosmeticEvents
    (:True
      (Utilities|IsValid PawnInCinematic
        (:Is%20Valid
          (if (Math|Boolean|ANDBoolean (Variables|Cinematic|Spectating|GetbAllowCinematicSpectating) (not (Variables|Cinematic|Spectating|GetbInCinematicSpectating)) (not (Utilities|IsValid (Variables|Cinematic|Spectating|GetSpectatorCinematicViewTarget) :!class /Script/Engine.KismetSystemLibrary)) (not (Utilities|IsValid (Variables|Cinematic|Spectating|GetPawnInCinematic) :!class /Script/Engine.KismetSystemLibrary)) (Utilities|IsValid (Variables|Cinematic|GetCinematicEnterSequence) :!class /Script/Engine.KismetSystemLibrary))
            (Variables|Cinematic|Spectating|SetbInCinematicSpectating true)
            (Variables|Cinematic|Spectating|SetbAllowCinematicSpectating false)
            ;;[1 1 0] Set PawnInCinematic and Create our list of spectators
            (bind _output_get (Variables|Cinematic|Spectating|SetPawnInCinematic PawnInCinematic))
            (Cinematic|Spectating|UpdateLocalSpectators _output_get)
            (if (Utilities|Array|IsNotEmpty (Variables|Cinematic|Spectating|GetLocalSpectators))
              (Utilities|FlowControl|Sequence
                (:then_0
                  (bind _returnvalue (Game|SpawnActorfromClass (Variables|Cinematic|GetCinematicViewTarget) (Math|Transform|MakeTransform _returnvalue_location (Math|Rotator|MakeRotator 0.0 0.0 (+ _returnvalue_rotation_yaw 1.8e+02))) "AlwaysSpawn" "SelectDefaultAtRuntime" 0 0))
                  (bind _output_get_2 (Variables|Cinematic|Spectating|SetSpectatorCinematicViewTarget _returnvalue))
                  (Cinematic|Spectating|SetSpectatorsViewTargetandAnimation (Variables|Cinematic|Spectating|GetLocalSpectators) _output_get_2 1.25 (Variables|Cinematic|GetCinematicEnterSequence)))
                (:then_1
                  (bind _returnvalue_2 (Utilities|Time|SetTimerbyEvent (EventDispatchers|CreateEvent :!event Custom|ApplySpectatingScreenWipe) (ScalableFloat|GetValueatLevel (Variables|RiftGate|Timing|InteractSequence|GetRow_Delay_ApplyPortalInteriorPostInteract))))
                  (Variables|Cinematic|Spectating|SetSpectatingScreenWipeTimer _returnvalue_2)
                  (bind _returnvalue_3 (Utilities|Time|SetTimerbyEvent (EventDispatchers|CreateEvent :!event Custom|CleanupCinematicSpectating) 8.0))
                  (Variables|Cinematic|Spectating|SetSpectatingCleanupFailsafeTimer _returnvalue_3))
                (:then_2
                  (Utilities|Time|SetTimerbyEvent (EventDispatchers|CreateEvent :!event Custom|PlayExitCinematicForSpectators) (+ (Variables|Cinematic|GetExitCinematicDelay) (Variables|Default|GetOpenRiftGateDelay)))
                  (for _array_element (Variables|Cinematic|Spectating|GetLocalSpectators)
                    (Cinematic|Spectating|CreateSpectatorCinematicBindings false _array_element)))))))))))

(event Custom|PlayDespawnSequence
  ;[1 1 0] Play Despawn Sequence
  (Utilities|FlowControl|DoOnce false
    (:Completed
      (Variables|Default|SetbDespawnCalled true)
      (CallFunction|SetRiftGateOpen false)
      (CallFunction|PlayDespawnAnticipationFX)
      (Utilities|Time|SetTimerbyEvent (EventDispatchers|CreateEvent :!event Custom|PlayDespawnFX) (ScalableFloat|GetValueatLevel (Variables|RiftGate|Timing|DespawnSequence|GetRow_Delay_DespawnFX)))
      (Utilities|Time|SetTimerbyEvent (EventDispatchers|CreateEvent :!event Custom|DespawnRiftGate) (ScalableFloat|GetValueatLevel (Variables|RiftGate|Timing|DespawnSequence|GetRow_Delay_Despawn))))))

(event Custom|StartHidingPostInteract
  ;[0.2 0.2 1] Gain Force Powers
  ;;[0 0 0] DEPRECATED - Moved to cinematic GA
  ;;;[0 0 0] StartHidingPostInteract
  (CallFunction|StartHiding (Variables|Default|GetLastPawnToInteract)))

(event EventInteractEnter
  ;[1 1 1] Event Interact Enter
  (if (Utilities|IsValid (Variables|Default|GetLastPawnToInteract) :!class /Script/Engine.KismetSystemLibrary)
    (GameplayEffects|FortApplyGameplayEffectToTarget :self (Ability|GetFortAbilitySystemComponent self) :GameplayEffectClass (Variables|Gameplay|GetGE_RiftGateEnter) :TargetOfEffect (Ability|GetFortAbilitySystemComponent (Variables|Default|GetLastPawnToInteract)) :OptionalSourceObject self :Level 0.0)))

(event Custom|DestroyBuildingsInArea
  ;[1 1 0] Destroy Buildings In Area
  (bind _outactors (Collision|BoxOverlapActors (Transformation|GetWorldTransform (Variables|Default|GetNoBuildZone)) (Components|Box|GetScaledBoxExtent (Variables|Default|GetNoBuildZone)) (Variables|UtilityVars|GetBuildingObjectTypes) "/Script/FortniteGame.BuildingActor" (Utilities|Array|MakeArray self) :!class /Script/FortniteGame.FortKismetLibrary))
  (for _array_element _outactors
    (bind _asbuilding_actor (Utilities|Casting|CastToBuildingActor _array_element))
    (if (Variables|/Script/Engine.Actor|GetbCanBeDamaged _asbuilding_actor)
      (Gameplay|Die :self _asbuilding_actor :DamageCauser self))))

(event Custom|KnockbackPlayersInArea
  (bind _self self)
  (bind _returnvalue (Ability|GetAbilitySystemComponent _self))
  ;[1 1 0] Knockback Players In Area
  (Utilities|IsValid _returnvalue
    (:Is%20Valid
      (Ability|Advanced|GrantandActivateAbilityonSelf _returnvalue (Variables|Knockback|GetGA_Knockback) _self))))

(event Custom|PlayExitCinematicForSpectators
  ;[0 0 0] Handle spectating exit cinematics
  (Networking|CanExecuteCosmeticEvents
    (:True
      (if (Math|Boolean|ANDBoolean (Utilities|IsValid (Variables|Cinematic|Spectating|GetPawnInCinematic) :!class /Script/Engine.KismetSystemLibrary) (Utilities|IsValid (Variables|Cinematic|GetCinematicExitSequence) :!class /Script/Engine.KismetSystemLibrary) (Variables|Cinematic|Spectating|GetbInCinematicSpectating))
        (Variables|Cinematic|Spectating|SetbInCinematicSpectating false)
        (Cinematic|Spectating|SetSpectatorsViewTargetandAnimation (Variables|Cinematic|Spectating|GetLocalSpectators) (Variables|Cinematic|Spectating|GetPawnInCinematic) 0.0 (Variables|Cinematic|GetCinematicExitSequence))
        (Utilities|FlowControl|Delay 0.2)
        (CallFunction|RemoveSpectatingScreenWipe)))))

(event Custom|SetRiftGateOpen
  (params (SetIsOpen :type Boolean))
  ;[1 0.8215 0] Set Rift Gate Open
  (if SetIsOpen
    (switch Utilities|FlowControl|Switch|SwitchonEnum_RiftGateState (Variables|RiftGate|State|GetRiftGateState_Current)
      (:NewEnumerator2
        _)
      (:NewEnumerator3
        _))
    (CallFunction|RiftGateSpawnAnim false)
    (Development|SetHiddeninGame (Variables|Default|GetSM_Pickle_Port_Inner) false true)
    (Niagara|SpawnSystematLocation "/RiftGateBase/Gameplay/FX/NS__RiftGate_SpawnImpact.NS__RiftGate_SpawnImpact" (FX|GetFXSpawnLocation) (0.0 0.0 0.0) "1.000000,1.000000,1.000000" true true "None" true)
    (Components|Activation|Deactivate (Variables|Default|GetNS_IdlePortal))
    (Utilities|FlowControl|Delay 1.0)
    (switch Utilities|FlowControl|Switch|SwitchonEnum_RiftGateState (Variables|RiftGate|State|GetRiftGateState_Current)
      (:NewEnumerator0
        (Variables|RiftGate|State|SetRiftGate_OpenAmount 1.0))
      (:NewEnumerator1
        _)
      (:NewEnumerator2
        _)
      (:NewEnumerator3
        _))
    (Variables|RiftGate|State|SetRiftGateState_Current "NewEnumerator1")
    (bind (_direction _openamount) (Timeline|OpenRiftGate Play)
      (:Update
        (Variables|RiftGate|State|SetRiftGate_OpenAmount _openamount))
      (:Finished
        (switch Utilities|FlowControl|Switch|SwitchonETimelineDirection _direction
          (:Forward
            (Variables|RiftGate|State|SetRiftGate_OpenAmount 1.0)
            (Variables|RiftGate|State|SetRiftGateState_Current "NewEnumerator0"))
          (:Backward
            (Variables|RiftGate|State|SetRiftGate_OpenAmount 0.0)
            (Variables|RiftGate|State|SetRiftGateState_Current "NewEnumerator2")
            (CallFunction|RiftGateSpawnAnim true)))))
    (else
      (switch Utilities|FlowControl|Switch|SwitchonEnum_RiftGateState (Variables|RiftGate|State|GetRiftGateState_Current)
        (:NewEnumerator0
          _)
        (:NewEnumerator1
          _))
      (Components|Timeline|SetPlayRate (Variables|RiftGateInteractBase|GetOpenRiftGate) (Math|Float|SafeDivide 1.0 (Variables|Default|GetCloseRiftGateDuration))) ; Currently closing faster than opening\nTo help match camera and pawn anim
      (switch Utilities|FlowControl|Switch|SwitchonEnum_RiftGateState (Variables|RiftGate|State|GetRiftGateState_Current)
        (:NewEnumerator0
          _)
        (:NewEnumerator1
          _)
        (:NewEnumerator2
          (Variables|RiftGate|State|SetRiftGate_OpenAmount 0.0))
        (:NewEnumerator3
          _))
      (Variables|RiftGate|State|SetRiftGateState_Current "NewEnumerator3")
      (Timeline|OpenRiftGate Reverse))))

(event Custom|KnockbackForCinematic
  (bind _self self)
  (bind _returnvalue (Ability|GetAbilitySystemComponent _self))
  ;[1 1 0] Knockback Players In Area Cinematic
  (Utilities|IsValid _returnvalue
    (:Is%20Valid
      (Ability|Advanced|GrantandActivateAbilityonSelf _returnvalue (Variables|Knockback|GetGA_KnockbackInCinematic) _self))))

(event Custom|OnSpectatorViewTargetChange
  (params (PlayerController :type Fort%20Player%20Controller%20Object%20Reference) (OldViewTarget :type Actor%20Object%20Reference) (NewViewTarget :type Actor%20Object%20Reference))
  ;[0 0 0] Spectating cinematics endplay and cleanup
  (if (not (or (== NewViewTarget (Variables|Cinematic|Spectating|GetSpectatorCinematicViewTarget)) (== NewViewTarget (Variables|Cinematic|Spectating|GetPawnInCinematic))))
    (Cinematic|Spectating|CleanupSpectator PlayerController)
    (Utilities|Array|RemoveItem (Variables|Cinematic|Spectating|GetLocalSpectators) PlayerController)))

(event Custom|RadialForce
  ;[1 1 0] Radial Force
  (bind _returnvalue (Development|GetConsoleVariableIntValue "r.PhysicsField.EnableField"))
  (if (== _returnvalue 1)
    (Game|SpawnActorfromClass "/Game/Effects/Niagara/Blueprints/PhsyicsFields/FN_RadialForce.FN_RadialForce_C" (Transformation|GetActorTransform) "Undefined" "MultiplyWithRoot" self 0.2 0.75 3e+03 1.5 0.0 0)))

(event Custom|UpdatePortalInterior
  ;[1 1 1] Update Portal Interior
  (bind (_direction _blend) (Timeline|TL_LightBlend PlayFromStart)
    (:Update
      (Rendering|Components|Light|SetLightColor (Variables|Lights|GetLight_Portal) (Math|Color|Lerp(LinearColor) (Variables|Lights|GetPortalLightColor_Start) (Variables|Lights|GetPortalLightColor_End) _blend) true)
      (Rendering|Material|SetScalarParameterValue (Variables|UtilityVars|GetMID_PortalInner) "InteriorBlend" _blend :!class /Script/Engine.MaterialInstanceDynamic))))

(event Custom|OnSpectatedPawnEndPlay
  (params (Actor :type Actor%20Object%20Reference) (EndPlayReason :type EEndPlayReason%20Enum :default "Destroyed"))
  ;[0 0 0] Spectating cinematics endplay and cleanup
  (CallFunction|CleanupCinematicSpectating))

(event Custom|OnCinematicViewTargetEndPlay
  (params (Actor :type Actor%20Object%20Reference) (EndPlayReason :type EEndPlayReason%20Enum :default "Destroyed"))
  ;[0 0 0] Spectating cinematics endplay and cleanup
  (CallFunction|CleanupCinematicSpectating))

(event Custom|PlayDespawnAnticipationFX
  ;[1 0 0.75] Play Despawn Anticipation FX - VFX, SFX, Camera Shake, Controller Rumble
  (bind _returnvalue (Niagara|SpawnSystematLocation (Variables|FX|Despawn|GetNS_DespawnAnticipationFX) (FX|GetFXSpawnLocation) (Math|Rotator|MakeRotator 0.0 0.0 (Math|Random|RandomFloatinRange -1.8e+02 1.8e+02)) "1.000000,1.000000,1.000000" true true "None" true))
  (Variables|Default|SetSpawned_DespawnAnticipationFX _returnvalue)
  (Niagara|SetNiagaraVariable(Float) (Variables|Default|GetSpawned_DespawnAnticipationFX) "Lifetime" (+ (ScalableFloat|GetValueatLevel (Variables|RiftGate|Timing|DespawnSequence|GetRow_Delay_DespawnFX)) 0.1))
  (LOD|SetMaxDrawDistance (Variables|Default|GetSpawned_DespawnAnticipationFX) 0.0)
  (CallFunction|TryStartCameraShake)
  (Components|Activation|Activate (Variables|Default|GetForceFeedback) false)
  (Utilities|FlowControl|Delay (ScalableFloat|GetValueatLevel (Variables|RiftGate|Timing|DespawnSequence|GetRow_Delay_Despawn)))
  (CameraShake|StopAllCameraShakes (Variables|Default|GetCameraShake) true)
  (Components|Activation|Deactivate (Variables|Default|GetForceFeedback))
  (Components|Activation|Deactivate (Variables|Default|GetSpawned_DespawnAnticipationFX)))

(event Custom|CleanupCinematicSpectating
  ;[0 0 0] Spectating cinematics endplay and cleanup
  (Utilities|FlowControl|Sequence
    (:then_0
      (for _array_element (Variables|Cinematic|Spectating|GetLocalSpectators)
        (Cinematic|Spectating|CleanupSpectator _array_element))
      (CallFunction|RemoveCachedControllersfromSpectatorArray))
    (:then_1
      (CallFunction|RemoveSpectatingScreenWipe)
      (Variables|Cinematic|Spectating|SetbAllowCinematicSpectating false)
      (Variables|Cinematic|Spectating|SetbInCinematicSpectating false))
    (:then_2
      (Utilities|Time|ClearandInvalidateTimerbyHandle (Variables|Cinematic|Spectating|GetSpectatingScreenWipeTimer))
      (Utilities|Time|ClearandInvalidateTimerbyHandle (Variables|Cinematic|Spectating|GetSpectatingCleanupFailsafeTimer)))
    (:then_3
      (bind _spectatorcinematicviewtarget (Variables|Cinematic|Spectating|GetSpectatorCinematicViewTarget)
        (:then
          (Actor|DestroyActor _spectatorcinematicviewtarget)
          (Variables|Cinematic|Spectating|SetSpectatorCinematicViewTarget 0))
        (:else)))
    (:then_4
      (Variables|Cinematic|Spectating|SetPawnInCinematic 0))))

(event Custom|DespawnRiftGate
  ;[1 1 1] Despawn Rift Gate
  (Components|Activation|Deactivate (Variables|Default|GetAthenaSpecialActor))
  (Collision|SetCollisionEnabled (Variables|Default|GetBox) "NoCollision")
  (Rendering|SetActorHiddenInGame true)
  (Variables|Interaction|SetbAllowInteract false)
  (if (Variables|Interaction|ControlVars|GetHasGrantedTraining)
    (CallFunction|OnRiftGateInteractionComplete)
    (else
      _))
  (if (Variables|RiftGate|State|GetShouldDieOnClose)
    (Gameplay|SilentDie false)))

(event Custom|PlayDespawnFX
  (bind (_returnvalue_roll _returnvalue_pitch _returnvalue_yaw) (Math|Rotator|BreakRotator (Transformation|GetWorldRotation (Variables|Default|GetNS_Pickle_Portal))))
  ;[1 0 0.75] Play DeSpawn FX
  (CallFunction|RiftGateSpawnAnim true)
  (Niagara|SpawnSystematLocation (Variables|FX|Despawn|GetNS_DespawnFX) (FX|GetFXSpawnLocation) (Math|Rotator|MakeRotator _returnvalue_roll _returnvalue_pitch _returnvalue_yaw) "1.000000,1.000000,1.000000" true true "None" true)
  (Audio|PlaySoundatLocation (Variables|Sound|Despawn|GetSound_DespawnFX) (FX|GetFXSpawnLocation))
  (Audio|Components|Audio|FadeOut (Variables|Default|GetSpawned_RiftPortalSoundLoop) 0.4 0.0 "Linear" :!class /Script/Engine.AudioComponent)
  (CallFunction|RadialForce)
  (ForceFeedback|SpawnForceFeedbackatLocation (Variables|FX|Despawn|GetForceFeedback_Despawn) (Transformation|GetActorLocation))
  (if (Variables|Default|GetCameraShakeEnabled)
    (Camera|PlayWorldCameraShake (Variables|FX|Despawn|GetCameraShake_Despawn) (Transformation|GetActorLocation) 0.0 5e+03)))

(event Custom|PlayCinematicSequence
  (bind _lastpawntointeract (Variables|Default|GetLastPawnToInteract))
  (bind _returnvalue (Ability|GetFortAbilitySystemComponent _lastpawntointeract))
  (bind _self self)
  (bind _returnvalue_2 (Ability|GetFortAbilitySystemComponent _self))
  ;[1 1 1] Play Cinematic Sequence
  (Utilities|FlowControl|Sequence
    (:then_0
      (GameplayEffects|RemoveActiveGameplayEffectbySourceEffect _returnvalue_2 "/RiftGateBase/Gameplay/Abilities/GE_Athena_RiftGateBase_DisableInput.GE_Athena_RiftGateBase_DisableInput_C" _returnvalue -1) ; Remove this when ga is consolidated
    )
    (:then_1
      (GameplayEffects|FortApplyGameplayEffectToTarget :self _returnvalue_2 :GameplayEffectClass (Variables|Cinematic|GetGE_CinematicSequence) :TargetOfEffect _returnvalue :OptionalSourceObject self :Level 0.0))))

(event Custom|ApplySpectatingScreenWipe
  (bind _screenwipe_gameplaycue (Variables|Cinematic|GetScreenWipe_GameplayCue))
  ;[0 0 0] Apply Screenwipe on spectator
  (bind _pawnincinematic (Variables|Cinematic|Spectating|GetPawnInCinematic)
    (:then
      (bind _returnvalue (Ability|GameplayCue|IsGameplayCueActive _pawnincinematic _screenwipe_gameplaycue))
      (if (not _returnvalue)
        (Projectile|AddGameplayCue%28Looping%29Local _pawnincinematic _screenwipe_gameplaycue (Ability|GameplayCue|MakeGameplayCueParameters 0.0))
        (Variables|Cinematic|Spectating|SetbSpectatingScreenwipeApplied true)))
    (:else)))

(event Custom|PlayRiftMusic
  (params (MusicStinger :type Sound%20Base%20Object%20Reference))
  ;[0.05185 0.1405 1] Plays Sequence Music on local player
  (bind _beinggrantedpowerspawn (Variables|RiftGate|Orientation|GetBeingGrantedPowersPawn)
    (:then
      (if (Pawn|IsLocallyViewed _beinggrantedpowerspawn)
        (Audio|AddEvent (GameInstanceSubsystems|GetMusicEventSubsystem) MusicStinger "(TagName=\"MusicEventSubsystem.Event.NPC\")" "(GameplayTags=((TagName=\"MusicEventSubsystem\")))")))
    (:else)))

(event Custom|RemoveSpectatingScreenWipe
  ;[0 0 0] Apply Screenwipe on spectator
  (Utilities|FlowControl|Delay 0.25)
  (bind _pawnincinematic (Variables|Cinematic|Spectating|GetPawnInCinematic)
    (:then
      (if (Variables|Cinematic|Spectating|GetbSpectatingScreenwipeApplied)
        (Projectile|RemoveGameplayCue%28Looping%29Local _pawnincinematic (Variables|Cinematic|GetScreenWipe_GameplayCue) (Ability|GameplayCue|MakeGameplayCueParameters 0.0))
        (Variables|Cinematic|Spectating|SetbSpectatingScreenwipeApplied false)))
    (:else)))

(event Custom|PlayLeaveMusic
  ;[0.09543 0.1731 1] Event for end music
  (CallFunction|PlayRiftMusic (Variables|Sound|Music|GetMusic_LeaveSequence)))

(event Custom|PlayRiftGateSpawnSequence
  ;[1 0 0.75] PlayRiftGateSpawnSequence
  (Niagara|SpawnSystemAttached "/RiftGateBase/Gameplay/FX/NS__RiftGate_SpawnImpact.NS__RiftGate_SpawnImpact" (Variables|Default|GetSM_Pickle_Port_Inner) "None" (FX|GetFXSpawnLocation)))

(event Custom|ResetRiftGateSpawn
  (bind _ns_portalzipper (Variables|Default|GetNS_PortalZipper))
  (bind _spawned_riftportalsoundloop (Variables|Default|GetSpawned_RiftPortalSoundLoop))
  ;[1 0 0.75] RiftGateSpawnAnim
  (Utilities|FlowControl|DoOnce (stale Custom|RiftGateSpawnAnim Despawning)
    (:Completed
      (bind _returnvalue (Niagara|SpawnSystematLocation "/RiftGateBase/Gameplay/FX/NS_RiftGateBase_Portal_Zipper.NS_RiftGateBase_Portal_Zipper" (Spline|GetLocationatSplinePoint (Variables|Default|GetPortal_Spline) (Math|Integer|SelectInt 4) "World") (Spline|GetRotationatSplinePoint (Variables|Default|GetPortal_Spline) (Math|Integer|SelectInt 4) "World")))
      (bind _output_get (Variables|Default|SetNS_PortalZipper _returnvalue))
      (Rendering|SetTranslucentSortPriority _output_get 2)
      (if (stale Custom|RiftGateSpawnAnim Despawning)
        (bind (_direction _spawninfluence _zipper_curve) (Timeline|SpawningAnim ReverseFromEnd)
          (:Update
            (Variables|RiftGate|State|SetRiftGate_Spawning _spawninfluence)
            (Utilities|IsValid _ns_portalzipper
              (:Is%20Valid
                (Transformation|SetWorldTransform _ns_portalzipper (Math|Transform|MakeTransform (Spline|GetLocationatDistanceAlongSpline (Variables|Default|GetPortal_Spline) (* _zipper_curve (Spline|GetSplineLength (Variables|Default|GetPortal_Spline))) "World") (Spline|GetRotationatDistanceAlongSpline (Variables|Default|GetPortal_Spline) (* _zipper_curve (Spline|GetSplineLength (Variables|Default|GetPortal_Spline))) "World")) false false))))
          (:Finished
            (switch Utilities|FlowControl|Switch|SwitchonETimelineDirection _direction
              (:Forward
                (Components|Activation|Deactivate (Variables|Default|GetForceFeedback)))
              (:Backward
                _))
            (CallFunction|ResetRiftGateSpawn)))
        (else
          (Audio|PlaySoundatLocation (Variables|Sound|Spawn|GetSound_SpawnRiftPortal) (Transformation|GetActorLocation))
          (Components|Activation|Activate (Variables|Default|GetForceFeedback) false)
          (Audio|Components|Audio|SetSound _spawned_riftportalsoundloop (Variables|Sound|Spawn|GetSound_SpawnRiftPortalLoop))
          (Audio|Components|Audio|FadeIn _spawned_riftportalsoundloop 0.3 1.0 0.0 "Linear" :!class /Script/Engine.AudioComponent)
          (Development|SetHiddeninGame (Variables|Lights|GetLight_Portal) false false)
          (Timeline|SpawningAnim PlayFromStart))))))

(event Custom|RiftGateSpawnAnim
  (params (Despawning :type Boolean))
  (bind _ns_portalzipper (Variables|Default|GetNS_PortalZipper))
  (bind _spawned_riftportalsoundloop (Variables|Default|GetSpawned_RiftPortalSoundLoop))
  ;[1 0 0.75] RiftGateSpawnAnim
  (Utilities|FlowControl|DoOnce Despawning
    (:Completed
      (bind _returnvalue (Niagara|SpawnSystematLocation "/RiftGateBase/Gameplay/FX/NS_RiftGateBase_Portal_Zipper.NS_RiftGateBase_Portal_Zipper" (Spline|GetLocationatSplinePoint (Variables|Default|GetPortal_Spline) (Math|Integer|SelectInt 4) "World") (Spline|GetRotationatSplinePoint (Variables|Default|GetPortal_Spline) (Math|Integer|SelectInt 4) "World")))
      (bind _output_get (Variables|Default|SetNS_PortalZipper _returnvalue))
      (Rendering|SetTranslucentSortPriority _output_get 2)
      (if Despawning
        (Timeline|SpawningAnim ReverseFromEnd)
        (else
          (Audio|PlaySoundatLocation (Variables|Sound|Spawn|GetSound_SpawnRiftPortal) (Transformation|GetActorLocation))
          (Components|Activation|Activate (Variables|Default|GetForceFeedback) false)
          (Audio|Components|Audio|SetSound _spawned_riftportalsoundloop (Variables|Sound|Spawn|GetSound_SpawnRiftPortalLoop))
          (Audio|Components|Audio|FadeIn _spawned_riftportalsoundloop 0.3 1.0 0.0 "Linear" :!class /Script/Engine.AudioComponent)
          (Development|SetHiddeninGame (Variables|Lights|GetLight_Portal) false false)
          (Timeline|SpawningAnim PlayFromStart))))))

(event Custom|OpenPortalAnim
  (params (ClosePortalGate :type Boolean) (AnimDuration :type Float%20%28double-precision%29) (LinearToEase :type Float%20%28double-precision%29) (InstantTransition :type Boolean))
  ;[1 0 0.75] OpenPortalAnim
  (Networking|CanExecuteCosmeticEvents
    (:True
      (Utilities|FlowControl|DoOnce false
        (:Completed
          (Rendering|Material|SetVectorParameterValue (Variables|UtilityVars|GetMID_PortalInner) "PPOW_AnimParams" (Utilities|Struct|MakeLinearColor (Math|Float|SelectFloat 0.0 (Utilities|Time|GetGameTimeinSeconds) InstantTransition) (Math|Conversions|ToFloat(Boolean) ClosePortalGate) AnimDuration LinearToEase) :!class /Script/Engine.MaterialInstanceDynamic)
          (Utilities|FlowControl|Delay AnimDuration))))))

(event Custom|DisableRiftgateCameraShake
  (Variables|Default|SetCameraShakeEnabled false))

(event Custom|PlotTwist_UseInstantCutEntranceCamera
  (params (Enable :type Boolean))
  (Variables|Default|SetUseInstantCutEntranceCameraTransition Enable))

(event Custom|TestCameraAnimation
  (bind _returnvalue (Fortnite|GetFirstLocalFortPlayerController true))
  (bind _returnvalue_2 (Pawn|GetControlledPawn _returnvalue))
  (bind _returnvalue_3 (Transformation|GetActorLocation _returnvalue_2))
  ;[0 0 0] TestCameraAnimation for exiting the rift gate
  (Networking|CanExecuteCosmeticEvents
    (:True
      (Utilities|IsValid _returnvalue
        (:Is%20Valid
          (bind _playercameramanager (Variables|/Script/Engine.PlayerController|GetPlayerCameraManager _returnvalue)
            (:then
              (Utilities|IsValid (CameraAnimation|ConvCameraAnimationCameraModifier _playercameramanager)
                (:Is%20Valid
                  (CameraAnimation|PlayCameraAnimation (CameraAnimation|ConvCameraAnimationCameraModifier _playercameramanager) "/RiftGateBase/Gameplay/Animation/Rift_Exit_Cam_Latest.Rift_Exit_Cam_Latest" (struct :PlayRate 1.0 :Scale 1.0 :EaseInType "Linear" :EaseInDuration 0.2 :EaseOutType "Linear" :EaseOutDuration 0.2 :bLoop false :StartOffset 0 :bRandomStartTime false :DurationOverride 0.0 :PlaySpace "CameraLocal" :UserPlaySpaceRot "0.000000,0.000000,0.000000") :!class /Script/EngineCameras.CameraAnimationCameraModifier)
                  (Development|PrintString (Utilities|String|Append "PawnLocation = " (Utilities|String|ToString(Vector) _returnvalue_3) " // " "CameraLocation = " (Utilities|String|ToString(Vector) (Camera|GetCameraLocation _playercameramanager))) true true "(R=1.000000,G=0.000000,B=0.750000,A=1.000000)" 6e+02)
                  (Rendering|Debug|DrawDebugSphere (Camera|GetCameraLocation _playercameramanager) 25.0 12 "(R=1.000000,G=1.000000,B=1.000000,A=1.000000)" 6e+02 3.0)
                  (Rendering|Debug|DrawDebugSphere _returnvalue_3 25.0 12 "(R=1.000000,G=1.000000,B=1.000000,A=1.000000)" 6e+02 3.0))))
            (:else)))))))

(fn AnimOpenPortalMID
  (params (Close%20Portal%20Gate :type Boolean) (Anim%20Duration :type Float%20%28double-precision%29 :default 1.0) (Linear%20to%20Ease :type Float%20%28double-precision%29 :default 1.0) (Instant%20Transition :type Boolean))
  (Rendering|Material|SetVectorParameterValue (Variables|UtilityVars|GetMID_PortalInner) "PPOW_AnimParams" (Utilities|Struct|MakeLinearColor (Math|Float|SelectFloat 0.0 (Utilities|Time|GetGameTimeinSeconds) Instant%20Transition) (Math|Conversions|ToFloat(Boolean) Close%20Portal%20Gate) Anim%20Duration Linear%20to%20Ease) :!class /Script/Engine.MaterialInstanceDynamic)
  (return))

(fn AnimPortalEnvTransitionMID
  (params (Reverse%20Animation :type Boolean) (Anim%20Duration :type Float%20%28double-precision%29 :default 1.0) (Instant%20Transition :type Boolean))
  (Development|PrintString (Utilities|String|Append "Transition Portal MID" (Utilities|String|ToString(Boolean) Reverse%20Animation) (Utilities|String|ToString(Float) Anim%20Duration) (Utilities|String|ToString(Boolean) Instant%20Transition)) true true "(R=0.152551,G=1.000000,B=0.000000,A=1.000000)" 5.0)
  (Rendering|Material|SetVectorParameterValue (Variables|UtilityVars|GetMID_PortalInner) "PPIMY_AnimParams" (Utilities|Struct|MakeLinearColor (Math|Float|SelectFloat 0.0 (Utilities|Time|GetGameTimeinSeconds) Instant%20Transition) (Math|Conversions|ToFloat(Boolean) Reverse%20Animation) Anim%20Duration) :!class /Script/Engine.MaterialInstanceDynamic)
  (return))

(fn BlueprintCanInteract
  (params (InteractingPawn :type Fort%20Pawn%20Object%20Reference) (InteractionBeingAttempted :type EInteractionBeingAttempted%20Enum :default "FirstInteraction") (InteractionType :type TInteractionType%20Enum :default "IT_NoInteraction"))
  (if (Math|Boolean|ANDBoolean (not (GameplayTags|HasAnyMatchingGameplayTags InteractingPawn (Variables|Tags|GetTags_InteractingPawnPreventInteract))) (not (Variables|Interaction|ControlVars|GetIsInteracting)) (Math|Float|InRange(Float) (Transformation|GetHorizontalDistanceTo (Variables|GetInteractingPawn) self) (ScalableFloat|GetValueatLevel (Variables|Interaction|GetRow_MinDistanceBetweenForInteract)) (ScalableFloat|GetValueatLevel (Variables|Interaction|GetRow_MaxDistanceBetweenForInteract))))
    (return (|CallParentFunction InteractingPawn InteractionBeingAttempted InteractionType))
    (else
      (return false))))

(fn BlueprintGetInteractionString
  (params (InteractingPawn :type Fort%20Pawn%20Object%20Reference) (InteractionBeingAttempted :type EInteractionBeingAttempted%20Enum :default "FirstInteraction"))
  (Utilities|IsValid InteractingPawn
    (:Is%20Valid
      (if (== (Utilities|Array|FindItem (Variables|Default|GetHidingPlayers) InteractingPawn) -1)
        (return (Variables|Interaction|GetInteractText))
        (else
          (return "NSLOCTEXT(\"[211D42129B0070D6846CC607D4E11DD5]\", \"CAE6E03E41C3388944D09AB0F8A9DCF5\", \"Exit\")"))))))

(fn BlueprintOnLocalInteract
  (params (InteractingPawn :type Fort%20Player%20Pawn%20Object%20Reference) (InteractionBeingAttempted :type EInteractionBeingAttempted%20Enum :default "FirstInteraction"))
  ;[0.3029 0.3432 1] If we have force powers already only show the deny message training.
  ; Returning false and not calling parent functionality because otherwise a hiding prop tag can get stuck on the local client which blocks many core gameplay actions.
  (if (GameplayTags|HasMatchingGameplayTag InteractingPawn (Variables|Tags|GetTag_PawnHasInteracted))
    (return false)
    (else
      ;[0.04319 0.04319 0.324] If we don't have force powers continue with normal flow of execution
      (bind _returnvalue (|CallParentFunction InteractingPawn InteractionBeingAttempted))
      (return _returnvalue))))

(fn BlueprintUpdateDynamicInteractionOptions
  (params (InteractingPawn :type Fort%20Pawn%20Object%20Reference) (InteractionBeingAttempted :type EInteractionBeingAttempted%20Enum :default "FirstInteraction") (InteractionType :type TInteractionType%20Enum :default "IT_NoInteraction"))
  (bind _outgameplaytags (|CallParentFunction InteractingPawn InteractionBeingAttempted InteractionType)))

(fn CanHidebyFalling
  (params (OtherActor :type Actor%20Object%20Reference))
  (return false 0))

(fn CheckForInSafeZone
  (Networking|SwitchHasAuthority
    (:Authority
      ;[1 0.3042 0.1474] If we passed Check 1
      ; Skip and go to Check 2
      (if (> (Variables|SafeZone|GetOutOfSafeZoneDestroyTime) 0)
        ;[0.1708 1 0.2319] Check 2
        (if (>= (Fortnite|Sync|Time|GetServerWorldTimeSeconds) (Variables|SafeZone|GetOutOfSafeZoneDestroyTime))
          ;[0.1708 1 0.2319] Check 3
          (bind _outfortplayerpawns (Fortnite|GetFortPlayerPawnsInRange (Transformation|GetActorLocation) 1e+03 true)) ; Are there any pawns close?\nThis should handle conversation and cinematic sequences
          (if (Utilities|Array|IsEmpty _outfortplayerpawns)
            ;[0 0 0] Do it
            (Variables|RiftGate|State|SetShouldDieOnClose true)
            (CallFunction|DespawnRiftGate)))
        (else
          ;[0.1708 1 0.2319] Check 1
          (bind _returnvalue (SafeZone|IsLocationinSafeZone (Transformation|GetActorLocation)))
          (if (not _returnvalue)
            (Variables|SafeZone|SetOutOfSafeZoneDestroyTime (+ (Fortnite|Sync|Time|GetServerWorldTimeSeconds) (Variables|SafeZone|GetOutOfSafeZoneLifetime)))))))))

(fn CleanupSpectator
  (params (LocalSpectator :type Fort%20Player%20Controller%20Object%20Reference))
  (Networking|CanExecuteCosmeticEvents
    (:True
      (Utilities|IsValid LocalSpectator
        (:Is%20Valid
          (Utilities|FlowControl|Sequence
            (:then_0
              (bind _playercameramanager (Variables|/Script/Engine.PlayerController|GetPlayerCameraManager LocalSpectator)
                (:then
                  (Utilities|IsValid (CameraAnimation|ConvCameraAnimationCameraModifier _playercameramanager)
                    (:Is%20Valid
                      (CameraAnimation|StopAllCameraAnimations (CameraAnimation|ConvCameraAnimationCameraModifier _playercameramanager) true :!class /Script/EngineCameras.CameraAnimationCameraModifier))))
                (:else)))
            (:then_1
              (Cinematic|Spectating|CreateSpectatorCinematicBindings true LocalSpectator))
            (:then_2
              (Utilities|Array|AddUnique (Variables|Cinematic|Spectating|GetSpectatorsToRemove) LocalSpectator))))))))

(fn CreateSpectatorCinematicBindings
  (params (bUnbind :type Boolean) (LocalFortPlayerControllerToBind :type Fort%20Player%20Controller%20Object%20Reference))
  (bind _outputdelegate (EventDispatchers|CreateEvent :!event OnSpectatorViewTargetChange))
  (bind _outputdelegate_2 (EventDispatchers|CreateEvent :!event OnCinematicViewTargetEndPlay))
  (bind _outputdelegate_3 (EventDispatchers|CreateEvent :!event OnSpectatedPawnEndPlay))
  (Networking|CanExecuteCosmeticEvents
    (:True
      (Utilities|FlowControl|Sequence
        (:then_0
          ;[0 0 0] Handle events for the local controller view target changing
          (bind _localfortplayercontrollertobind (Variables|GetLocalFortPlayerControllerToBind)
            (:then
              (if (not (Variables|GetbUnbind))
                (EventDispatchers|BindEventtoOnViewTargetChanged _localfortplayercontrollertobind _outputdelegate) ; Binds exec path
                (else
                  (EventDispatchers|UnbindEventfromOnViewTargetChanged _localfortplayercontrollertobind _outputdelegate) ; Unbinds exec path
                )))
            (:else)))
        (:then_1
          ;[0 0 0] Handle events for EndPlay of the PawnInCinematic
          (bind _pawnincinematic (Variables|Cinematic|Spectating|GetPawnInCinematic)
            (:then
              (if (not (Variables|GetbUnbind))
                (Game|BindEventtoOnEndPlay _pawnincinematic _outputdelegate_3)
                (else
                  (Game|UnbindEventfromOnEndPlay _pawnincinematic _outputdelegate_3))))
            (:else)))
        (:then_2
          ;[0 0 0] Handle events for EndPlay of the SpectatorCinematicViewTarget Actor
          (bind _spectatorcinematicviewtarget (Variables|Cinematic|Spectating|GetSpectatorCinematicViewTarget)
            (:then
              (if (not (Variables|GetbUnbind))
                (Game|BindEventtoOnEndPlay _spectatorcinematicviewtarget _outputdelegate_2)
                (else
                  (Game|UnbindEventfromOnEndPlay _spectatorcinematicviewtarget _outputdelegate_2))))
            (:else)))))))

(fn GetFXSpawnLocation
  (return (+ (Transformation|GetActorLocation) (Math|Vector|RotateVector (Variables|FX|GetFXOffset) (Transformation|GetActorRotation)))))

(fn HandleCustomImpulseFromDamage
  (params (ImpulseValue :type Vector :default "0, 0, 0") (Damage :type Float%20%28single-precision%29) (DamageTags :type Gameplay%20Tag%20Container%20Structure%20%28by%20ref%29) (HitResult :type Hit%20Result%20Structure%20%28by%20ref%29) (InstigatingController :type Controller%20Object%20Reference) (DamageCauser :type Actor%20Object%20Reference) (EffectContext :type Gameplay%20Effect%20Context%20Handle%20Structure%20%28by%20ref%29))
  (return false))

(fn IsControllerSpectatingActor
  (params (InFortPlayerController :type Fort%20Player%20Controller%20Object%20Reference) (InActor :type Actor%20Object%20Reference))
  ;[0.2 1 0.2] Validate parameters
  (if (and (Utilities|IsValid InFortPlayerController :!class /Script/Engine.KismetSystemLibrary) (Utilities|IsValid InActor :!class /Script/Engine.KismetSystemLibrary))
    ;[0 0 0] Check to see if the InActor is a pawn
    (bind _asfort_player_pawn (Utilities|Casting|CastToFortPlayerPawn (Variables|GetInActor))
      (:then
        (bind _output_get (LocalVariables|SetLInActorsPawn _asfort_player_pawn))
        ;[0 0 0] Checks the controller we are testing to see if it's pawn is the same as the the InActors
        ; Also checks to see if the InActor is our view target
        (return (and (select (Utilities|IsValid (Pawn|GetControlledPawn (Variables|GetInFortPlayerController)) :!class /Script/Engine.KismetSystemLibrary) (!= _output_get (Pawn|GetControlledPawn (Variables|GetInFortPlayerController))) true) (== (Pawn|GetViewTarget (Variables|GetInFortPlayerController)) (Variables|GetInActor)))))
      (:CastFailed
        ;[0 0 0] Checks the InController to see if the InActor is the viewtarget
        (return (== (Pawn|GetViewTarget (Variables|GetInFortPlayerController)) (Variables|GetInActor)))))
    (else
      (return false))))

(fn IsControllerSpectatingPawn
  (params (LocalController :type Fort%20Player%20Controller%20Object%20Reference) (SpectatedPawn :type Fort%20Player%20Pawn%20Object%20Reference))
  (if (not (and (Utilities|IsValid LocalController :!class /Script/Engine.KismetSystemLibrary) (Utilities|IsValid SpectatedPawn :!class /Script/Engine.KismetSystemLibrary)))
    (return false)))

(fn MarkActor
  (params (ActorToMark :type Actor%20Object%20Reference))
  (if (Variables|Default|GetShould%20Mark%20Rift)
    (for _array_element (Fortnite|GetAllFortPlayerControllers true true)
      (Utilities|IsValid _array_element
        (:Is%20Valid
          (Utilities|IsValid (Class|FortPlayerController|GetMarkerComponent _array_element)
            (:Is%20Valid
              (Utilities|IsValid ActorToMark
                (:Is%20Valid
                  (bind _returnvalue (MarkerComponent|MarkActoronServer (Class|FortPlayerController|GetMarkerComponent _array_element) ActorToMark false false))
                  (Variables|Default|SetPortalMarkerID _returnvalue))))))))))

(fn OnRep_BeingGrantedPowersPawn
  (Portal|SetHidingPropAnimsbyGender (Variables|RiftGate|Orientation|GetBeingGrantedPowersPawn)))

(fn OnRep_bPlayerHasExitedPortal
  (bind (_returnvalue_location _returnvalue_rotation) (Math|Transform|BreakTransform (Transformation|GetWorldTransform (Variables|Default|GetHideLocation_ForwardVector))))
  (bind (_returnvalue_rotation_roll _returnvalue_rotation_pitch _returnvalue_rotation_yaw) (Math|Rotator|BreakRotator _returnvalue_rotation))
  (bind _beinggrantedpowerspawn (Variables|RiftGate|Orientation|GetBeingGrantedPowersPawn))
  (if (Variables|Default|GetbPlayerHasExitedPortal)
    (if (not (Pawn|IsLocallyControlled _beinggrantedpowerspawn))
      (Transformation|SetActorRotation _beinggrantedpowerspawn (Math|Rotator|MakeRotator _returnvalue_rotation_roll _returnvalue_rotation_pitch _returnvalue_rotation_yaw) false))))

(fn OnRep_RiftGate_OpenAmount
  (bind _riftgate_openamount (Variables|RiftGate|State|GetRiftGate_OpenAmount))
  (bind _param_openamount (Variables|RiftGate|State|GetParam_OpenAmount))
  (Rendering|Material|SetScalarParameterValueonMaterials (Variables|Default|GetSM_Pickle_Port_Inner) _param_openamount _riftgate_openamount)
  (Niagara|SetNiagaraVariable(Float) (Variables|Default|GetNS_Pickle_Portal) _param_openamount _riftgate_openamount))

(fn OnRep_RiftGate_Spawning
  (bind _riftgate_spawning (Variables|RiftGate|State|GetRiftGate_Spawning))
  (Rendering|Material|SetScalarParameterValueonMaterials (Variables|Default|GetSM_Pickle_Port_Outer) "SpawnInfluence" _riftgate_spawning)
  (Niagara|SetNiagaraVariable(Float) (Variables|Default|GetNS_Pickle_Portal) "MasterOpacity" _riftgate_spawning)
  (Rendering|Components|Light|SetIntensity (Variables|Lights|GetLight_Portal) (* _riftgate_spawning (Variables|Lights|GetPortalLightIntensity)))
  (Rendering|Lighting|SetAttenuationRadius (Variables|Lights|GetLight_Portal) (* (Variables|Lights|GetPortalLightAttenuation) _riftgate_spawning))
  (Rendering|Components|Light|SetIntensity (Variables|Lights|GetLight_Ambient) (* _riftgate_spawning (Variables|Lights|GetAmbientLightIntensity))))

(fn OnRep_RiftGateState_Current
  (switch Utilities|FlowControl|Switch|SwitchonEnum_RiftGateState (Variables|RiftGate|State|GetRiftGateState_Current)
    (:NewEnumerator0
      (CallFunction|OpenPortalAnim false 1.0 1.0 true))
    (:NewEnumerator1
      (CallFunction|OpenPortalAnim false 1.0 1.0))
    (:NewEnumerator2
      (CallFunction|OpenPortalAnim true 1.0 1.0 true))
    (:NewEnumerator3
      (CallFunction|OpenPortalAnim true 0.9 1.0))))

(fn OnRiftGateInteractionComplete)

(fn RemoveCachedControllersFromSpectatorArray
  (for _array_element (Variables|Cinematic|Spectating|GetSpectatorsToRemove)
    (Utilities|Array|RemoveItem (Variables|Cinematic|Spectating|GetLocalSpectators) _array_element))
  (Utilities|Array|Clear (Variables|Cinematic|Spectating|GetSpectatorsToRemove)))

(fn SetHidingPropAnimsByGender
  (params (FortPawn :type Fort%20Pawn%20Object%20Reference))
  (bind _femaleentermontage (Variables|PawnAnimation|GetFemaleEnterMontage))
  (bind _maleentermontage (Variables|PawnAnimation|GetMaleEnterMontage))
  (bind _femaleexitmontage (Variables|PawnAnimation|GetFemaleExitMontage))
  (bind _maleexitmontage (Variables|PawnAnimation|GetMaleExitMontage))
  (bind _asfort_player_pawn (Utilities|Casting|CastToFortPlayerPawn FortPawn))
  (Utilities|FlowControl|Sequence
    (:then_0
      (Utilities|IsValid (select (FortPlayerPawn|GetGender _asfort_player_pawn) _maleentermontage _femaleentermontage)
        (:Is%20Valid
          (Variables|Animation|SetEnterAnimMontage (select (FortPlayerPawn|GetGender _asfort_player_pawn) _maleentermontage _femaleentermontage)))))
    (:then_1
      (Utilities|IsValid (select (FortPlayerPawn|GetGender _asfort_player_pawn) _maleexitmontage _femaleexitmontage)
        (:Is%20Valid
          (Variables|Animation|SetExitAnimMontage (select (FortPlayerPawn|GetGender _asfort_player_pawn) _maleexitmontage _femaleexitmontage)))))))

(fn SetSpectatorsViewTargetAndAnimation
  (params (LocalControllers :type Array%20of%20Fort%20Player%20Controller%20Object%20References) (InNewViewTarget :type Actor%20Object%20Reference) (InBlendTime :type Float%20%28double-precision%29) (InCameraAnimation :type Camera%20Animation%20Sequence%20Object%20Reference))
  (for _array_element LocalControllers
    (Utilities|IsValid _array_element
      (:Is%20Valid
        (bind _output_get (LocalVariables|SetLCurrentController _array_element))
        (Utilities|FlowControl|Sequence
          (:then_0
            (bind _innewviewtarget (Variables|GetInNewViewTarget)
              (:then
                (Game|Player|SetViewTargetwithBlend _output_get _innewviewtarget (Variables|GetInBlendTime) "VTBlend_Linear" 0.0 true))
              (:else)))
          (:then_1
            (bind _playercameramanager (Variables|/Script/Engine.PlayerController|GetPlayerCameraManager _output_get)
              (:then
                (bind _output_get_2 (LocalVariables|SetLCurrentCameraModifier (CameraAnimation|ConvCameraAnimationCameraModifier _playercameramanager)))
                (Utilities|IsValid _output_get_2
                  (:Is%20Valid
                    (bind _incameraanimation (Variables|GetInCameraAnimation)
                      (:then
                        (CameraAnimation|PlayCameraAnimation _output_get_2 _incameraanimation (Utilities|Struct|MakeCameraAnimationParams 1.0) :!class /Script/EngineCameras.CameraAnimationCameraModifier))
                      (:else
                        (CameraAnimation|StopAllCameraAnimations _output_get_2 true :!class /Script/EngineCameras.CameraAnimationCameraModifier))))))
              (:else))))))))

(fn StartInitialSpawnAnticipationFX
  (bind _ns_spawnanticipationfx (Variables|FX|Spawn|GetNS_SpawnAnticipationFX)
    (:then
      (bind _returnvalue (Niagara|SpawnSystematLocation _ns_spawnanticipationfx (FX|GetFXSpawnLocation) (Math|Rotator|MakeRotator 0.0 0.0 (Math|Random|RandomFloatinRange -1.8e+02 1.8e+02)) "1.000000,1.000000,1.000000" true true "None" true))
      (bind _output_get (Variables|Default|SetSpawned_SpawnAnticipationFX _returnvalue))
      (Utilities|IsValid _output_get
        (:Is%20Valid
          (Niagara|SetNiagaraVariable(Float) _output_get "Lifetime" (+ (ScalableFloat|GetValueatLevel (Variables|RiftGate|Timing|SpawnSequence|GetRow_Duration_InitialSpawnAnticipationFX)) 0.025))
          (Audio|PlaySoundatLocation (Variables|Sound|Spawn|GetSound_SpawnAnticipationFX) (FX|GetFXSpawnLocation))
          (bind _returnvalue_2 (Audio|SpawnSoundatLocation (Variables|Sound|Music|GetMusic_SpawnAnticipation) (FX|GetFXSpawnLocation)))
          (Audio|AddEventbyComponent (GameInstanceSubsystems|GetMusicEventSubsystem) _returnvalue_2 "(TagName=\"MusicEventSubsystem.Event.POI\")" "(GameplayTags=((TagName=\"MusicEventSubsystem\")))" 0 "")
          (SoundIndicator|StartSoundIndicatorLooping (Variables|Default|GetFortSoundIndicator) (FX|GetFXSpawnLocation) (Variables|Sound|Spawn|GetSound_SpawnAnticipationFX) 1.0 0)
          (CallFunction|TryStartCameraShake)
          (Components|Activation|Activate (Variables|Default|GetForceFeedback) false))))
    (:else)))

(fn StopInitialSpawnAnticipationFX
  (SoundIndicator|StopSoundIndicatorLooping (Variables|Default|GetFortSoundIndicator))
  (CameraShake|StopAllCameraShakes (Variables|Default|GetCameraShake) true)
  (Components|Activation|Deactivate (Variables|Default|GetForceFeedback))
  (Components|Activation|Deactivate (Variables|Default|GetSpawned_SpawnAnticipationFX)))

(fn TryAddSpecialActorRenderTag
  (params (bForceRemove :type Boolean))
  (Networking|CanExecuteCosmeticEvents
    (:True
      (Fortnite|FortShippingLog self "Called TryAddSpecialActorRenderTag")
      (for _array_element (Fortnite|GetLocalFortPlayerControllers true)
        (bind _asfort_player_pawn (Utilities|Casting|CastToFortPlayerPawn (Pawn|GetControlledPawn _array_element))
          (:then
            (if (Variables|GetbForceRemove)
              ;[0 0 0] Force removes the render tag
              (Ability|GameplayEffect|RemoveLooseGameplayTags _asfort_player_pawn (Variables|/Script/FortniteGame.AthenaSpecialActorComponent|GetRenderTags (Variables|Default|GetAthenaSpecialActor)))
              ;[1 1 0] Logging
              (Fortnite|FortShippingLog self (Utilities|String|Append "Removed Rift Gate Render Tag for " (Fortnite|GetPlayerNameSafe _asfort_player_pawn)))
              (else
                ;[0 0 0] Adds the RenderTag for the special actor component if not already present
                (if (not (GameplayTags|HasAnyMatchingGameplayTags _asfort_player_pawn (LocalVariables|GetL_TagsToSearchBeforeApplyingRenderTag)))
                  (Ability|GameplayEffect|AddLooseGameplayTags _asfort_player_pawn (Variables|/Script/FortniteGame.AthenaSpecialActorComponent|GetRenderTags (Variables|Default|GetAthenaSpecialActor)))
                  (Fortnite|FortShippingLog self (Utilities|String|Append "Added Rift Gate Render Tag for " (Fortnite|GetPlayerNameSafe _asfort_player_pawn)))))))
          (:CastFailed
            ;[1 0 0] ControlledPawn cast to FortPlayerPawn Failed
            (Fortnite|FortShippingLog self "ControlledPawn cast to FortPlayerPawn Failed" true)))))))

(fn TryStartCameraShake
  (if (Variables|Default|GetCameraShakeEnabled)
    (CameraShake|Start (Variables|Default|GetCameraShake))))

(fn UnmarkActor
  (bind _slatebrush (Utilities|Struct|MakeSlateBrush "(SpecifiedColor=(R=1.000000,G=1.000000,B=1.000000,A=1.000000),ColorUseRule=UseColor_Specified)" "Image" "NoTile" "NoMirror" (struct :X 0.0 :Y 0.0) "(Left=0.000000,Top=0.000000,Right=0.000000,Bottom=0.000000)" 0 "(CornerRadii=(X=0.000000,Y=0.000000,Z=0.000000,W=0.000000),Color=(SpecifiedColor=(R=0.000000,G=0.000000,B=0.000000,A=0.000000),ColorUseRule=UseColor_Specified),Width=0.000000,RoundingType=HalfHeightRadius,bUseBrushTransparency=False)"))
  (bind _athenaspecialactor (Variables|Default|GetAthenaSpecialActor))
  (if (Variables|Default|GetShould%20Mark%20Rift)
    (for _array_element (Fortnite|GetAllFortPlayerControllers true true)
      (Utilities|IsValid _array_element
        (:Is%20Valid
          (Utilities|IsValid (Class|FortPlayerController|GetMarkerComponent _array_element)
            (:Is%20Valid
              (MarkerComponent|UnmarkActoronClient (Class|FortPlayerController|GetMarkerComponent _array_element) (Variables|Default|GetPortalMarkerID)))))))
    (Variables|/Script/FortniteGame.AthenaSpecialActorComponent|SetSpecialActorCompassIconBrush _slatebrush _athenaspecialactor) ; Remove compass icon
    (Variables|/Script/FortniteGame.AthenaSpecialActorComponent|SetSpecialActorMinimapIconBrush _slatebrush _athenaspecialactor) ; Remove minimap icon
  ))

(fn UpdateLocalSpectators
  (params (SpectatedActor :type Actor%20Object%20Reference))
  (bind _localspectators (Variables|Cinematic|Spectating|GetLocalSpectators))
  (Networking|CanExecuteCosmeticEvents
    (:True
      (Utilities|IsValid SpectatedActor
        (:Is%20Valid
          (Utilities|Array|Clear _localspectators)
          (for _array_element (Fortnite|GetLocalFortPlayerControllers true true)
            (Utilities|IsValid _array_element
              (:Is%20Valid
                (if (Cinematic|Spectating|IsControllerSpectatingActor _array_element (Variables|GetSpectatedActor))
                  (Utilities|Array|AddUnique _localspectators _array_element))))))))))

(fn ConstructionScript
  (|CallParentFunction)
  (Rendering|Components|Light|SetLightColor (Variables|Lights|GetLight_Portal) (Variables|Lights|GetPortalLightColor_Start) true)
  (bind _returnvalue (Rendering|Material|CreateDynamicMaterialInstance (Variables|Default|GetSM_Pickle_Port_Inner) 0 (Variables|UtilityVars|GetInnerPortalMaterial) "None" :!class /Script/Engine.PrimitiveComponent))
  (Variables|UtilityVars|SetMID_PortalInner _returnvalue)
  (Rendering|Material|CreateDynamicMaterialInstance (Variables|Default|GetSM_Pickle_Port_Back) 0 0 "None" :!class /Script/Engine.PrimitiveComponent)
  (if (Variables|RiftGate|State|GetIsOpen)
    _)
  (CallFunction|SetRiftGateOpen (Variables|RiftGate|State|GetIsOpen))
  (Components|Timeline|SetNewTime (Variables|RiftGateInteractBase|GetOpenRiftGate) (Math|Conversions|ToFloat(Boolean) (Variables|RiftGate|State|GetIsOpen))))
