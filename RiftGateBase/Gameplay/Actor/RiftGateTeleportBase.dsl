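; Base rift gate actor that teleports an eligible player to a hotfix-set height at the gate's XY and starts skydiving.
; PerformInteraction rejects DBNO or tag-blocked pawns, blocks disguise interactions and ejects vehicle occupants.
; The authority detaches remote control, drops carried players, ignores fall damage, multicasts rift VFX, then, unless IsFlying, teleports and knocks players back.
; EventCreateInteractionFX plays delayed exit audio when the overlapping actor isn't the local player.
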(event EventCreateInteractionFX
  (params (Location :type Vector :default "0, 0, 0"))
  ;[1 1 1] Spawn teleport visuals and audio for clients
  (|CallParentFunction Location)
  (if (not (Utilities|Equal(Object) (Variables|RiftPortal|GetOverlapActor) (Game|GetPlayerPawn 0)))
    (Utilities|FlowControl|Delay (ScalableFloat|GetValueatLevel (Variables|HotfixRows|GetDelayExitAfterTeleportRow))) ; Delay Before FX
    (Audio|SpawnSoundatLocation (Variables|RiftPortal|GetPlayerExitSkyCrack) (Variables|RiftPortal|GetTeleportLocation))))

(event Custom|SpawnRiftEffectsMulticast
  (params (Location :type Vector))
  (Niagara|SpawnSystematLocation "/RiftGateBase/VFX/NS_Core_Rift_Open.NS_Core_Rift_Open" Location "90.000000,0.000000,0.000000"))

(fn CanPawnBeTeleported
  (params (InputPin :type Fort%20Player%20Pawn%20Object%20Reference))
  (return (and (not (Variables|/Script/FortniteGame.FortPawn|GetbIsDBNO InputPin)) (not (GameplayTags|HasAnyMatchingGameplayTags InputPin (Variables|RiftPortal|GetPawnBlockTeleportTags))))))

(fn PerformInteraction
  (bind _playerpawn (Variables|RiftPortal|GetPlayerPawn))
  (bind _playerpawn_2 (Variables|RiftPortal|GetPlayerPawn))
  (bind _succeeded (|CallParentFunction))
  (if _succeeded
    (Variables|RiftPortal|SetActorRotation (Transformation|GetActorRotation (Variables|RiftPortal|GetOverlapActor)))
    (Utilities|FlowControl|Sequence
      (:then_0
        (Utilities|FlowControl|DoOnce false
          (:Completed
            (Variables|RiftPortal|SetTeleportLocation (Math|Vector|vector+vector (* (Transformation|GetActorLocation) "1.000000,1.000000, 0") (Math|Vector|MakeVector 0.0 0.0 (ScalableFloat|GetValueatLevel (Variables|HotfixRows|GetTeleportHeightRow))))))))
      (:then_1
        (if (Default|CanPawnbeTeleported (Variables|RiftPortal|GetPlayerPawn))
          (CallFunction|BlockInteractionsWIthDisguises 0)
          (if (Vehicle|IsinVehicle (Variables|RiftPortal|GetPlayerPawn))
            (Utilities|IsValid (Utilities|ToObject(Interface) (Vehicle|GetVehicleInterface (Variables|RiftPortal|GetPlayerPawn)))
              (:Is%20Valid
                (GameplayEffects|RemoveActiveEffectswithGrantedTags (Ability|GetAbilitySystemComponent (Variables|RiftPortal|GetPlayerPawn)) "(GameplayTags=((TagName=\"Gameplay.Action.Player.AthenaInVehicle\")))") ; Eject Player from vehicle before teleporting
              ))
            (else
              _))
          (Networking|SwitchHasAuthority
            (:Authority
              (RemoteControlPawn|ForceDetachfromRemoteControlledPawn _playerpawn)
              (Class|FortPlayerPawn|StopCarryingPlayer _playerpawn false)
              (Variables|/Script/FortniteGame.FortPawn|SetbIgnoreNextFallingDamage true _playerpawn) ; Have to do this because of the delay. Otherwise player can die from fall damage if they fall from high distance.
              (CallFunction|SpawnRiftEffectsMulticast (Variables|RiftPortal|GetTeleportLocation))
              (bind _returnvalue (CallFunction|IsFlying))
              (if (not _returnvalue)
                (Parachute|EndSkydiving _playerpawn_2) ; End Skydiving if player was previously
                (CallFunction|TeleportPlayer _playerpawn_2)
                (Variables|Default|SetInteractionPerformed true)
                (CallFunction|KnockBackPlayers)
                (return true)))))))
    (else
      (return false))))

(fn TeleportPlayer
  (params (Player :type Fort%20Player%20Pawn%20Object%20Reference))
  (bind _returnvalue (Transform|Actor|TeleportPlayerPawn Player (Variables|RiftPortal|GetTeleportLocation)))
  (if _returnvalue
    (Parachute|BeginSkydiving Player false)))

(fn ConstructionScript
  (|CallParentFunction))
